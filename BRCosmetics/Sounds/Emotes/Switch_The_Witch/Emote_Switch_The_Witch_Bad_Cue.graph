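# Sound cue for the "Bad" music variant of the Switch The Witch emote.
# A single wave player feeds Emote_Switch_The_Witch_Bad_01 straight to the output, with no randomization or mixing nodes.
# It uses the Fort_Music_Emote sound class and the shared Athena emote-music concurrency rule.
# A submix send also routes it to Fortnite_AudioAnalysis_Music, which by its name serves music audio analysis.
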SoundCue /BRCosmetics/Sounds/Emotes/Switch_The_Witch/Emote_Switch_The_Witch_Bad_Cue
root=WavePlayer_0
nodes [{id=WavePlayer_0, properties {soundWaveAssetPtr=&/Emote_Switch_The_Witch_Bad_01}}]
properties {
 maxDistance=2097152
 soundClassObject=&Fort_Music_Emote
 volumeMultiplier=1
 concurrencyOverrides {maxCount=3, resolutionRule=StopOldest}
 concurrencySet [/Game/Sounds/Concurrency_Rules/Athena_EmoteMusic_Conc]
 soundSubmixSends [{minSendLevel=0, soundSubmix=&Fortnite_AudioAnalysis_Music}]
}
